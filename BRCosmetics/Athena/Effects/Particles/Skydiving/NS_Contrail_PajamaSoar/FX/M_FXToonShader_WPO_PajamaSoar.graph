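# Toon shader material for the PajamaSoar skydiving contrail's Niagara meshes and ribbons, using the Cloth shading model with world-position-offset animation.
# A curve atlas blends shaded and base textures, with switchable ink lines, highlights, reflection and rim effects, output as time-of-day-adjusted emissive.
# Layered sine waves along instance-space Z displace vertices, growing with height and driven by dynamic parameters.
# A previous-frame switch offsets the wave phase for motion vectors.
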Material /BRCosmetics/Athena/Effects/Particles/Skydiving/NS_Contrail_PajamaSoar/FX/M_FXToonShader_WPO_PajamaSoar
graph {
 properties {
  bUseMaterialAttributes=true
  bUsedWithNiagaraMeshParticles=true
  bUsedWithNiagaraRibbons=true
  pixelDepthOffsetMode=PDOM_Legacy
  shadingModel=MSM_Cloth
  usedShadingModels=Cloth
 }
 nodes [
  n1 = DynamicParameter() {defaultValue=<rgba 0 1 1 2>, paramNames [Time Offset, Speed, Pulse, Frequency]}
  n2 = Multiply(A=@n1.Frequency)
  n3 = TransformPosition(None=WorldPosition().XYZ) {transformSourceType=TRANSFORMPOSSOURCE_World, transformType=TRANSFORMPOSSOURCE_Instance}
  n4 = Saturate(None=Divide(A=ComponentMask(None=@n3) {b=true}, B=ScalarParameter() {defaultValue=200, group=WPO, parameterName=WPO Object Size}) {constB=200})
  n5 = Power(Base=@n4, Exp=ScalarParameter() {defaultValue=16, group=WPO, parameterName=Power Hand Mask}) {constExponent=12}
  n6 = ScalarParameter() {defaultValue=7, group=WPO, parameterName=WPO Speed}
  n7 = Multiply(A=@n6) {constB=0.8}
  n8 = DeltaTime()
  n9 = LinearInterpolate(Alpha=Saturate(None=@n1.Pulse)) {constA=1, constB=2}
  n10 = PreviousFrameSwitch(Current Frame=@n1.Speed, Previous Frame=Subtract(A=@n1.Speed, B=Multiply(A=@n8, B=@n9)))
  n11 = Subtract(
   A=Multiply(A=LinearInterpolate(A=@n1.Frequency, B=@n2, Alpha=@n5) {constA=1, constB=4}, B=@n4)
   B=Add(A=@n1.Time Offset, B=Multiply(A=@n7, B=@n10))
  )
  Cosine(None=@n11)
  n12 = Multiply(A=@n1.Frequency) {constB=3}
  n13 = Multiply(A=@n6) {constB=0.9}
  n14 = Subtract(
   A=Multiply(A=LinearInterpolate(A=@n1.Frequency, B=@n12, Alpha=@n5) {constA=1, constB=4}, B=@n4)
   B=Add(A=@n1.Time Offset, B=Multiply(A=@n13, B=@n10))
  )
  Cosine(None=@n14)
  n15 = NamedRerouteUsage()
  n16 = NamedRerouteUsage()
  n17 = ScalarParameter() {group="##31.Highlight", parameterName=HighLight_BlendMethod}
  LinearInterpolate(
   A=LinearInterpolate(A=Add(A=@n15, B=@n16), B=@n16, Alpha=Saturate(None=@n17))
   B=Multiply(A=@n15, B=@n16)
   Alpha=Saturate(None=Subtract(A=@n17))
  )
  n18 = Constant()
  n19 = TextureSampleParameter2D() {
   group="##01.Textures"
   parameterName=Texture_SSC
   samplerSource=SSM_Wrap_WorldGroupSettings
   samplerType=SAMPLERTYPE_Masks
   sortPriority=1
   texture=/BRCosmetics/Athena/Effects/Particles/Skydiving/NS_Contrail_PajamaSoar/Textures/T_PajamaSoar_Contrail_STT
  }
  n20 = NamedRerouteUsage()
  n21 = ParticleColor()
  n22 = Constant() {r=1}
  n23 = If(A=@n21.A, "A > B"=@n21.A, "A == B"=@n22, "A < B"=@n22)
  n24 = VectorParameter() {group="########BasicSetting", parameterName=BaseColor, defaultValue=<rgba 1 1 1 1>}
  n25 = TextureSampleParameter2D() {
   group="##01.Textures"
   parameterName=Texture_ShadedDiffuse
   sortPriority=1
   texture=/BRCosmetics/Athena/Effects/Particles/Skydiving/NS_Contrail_PajamaSoar/Textures/T_PajamaSoar_Contrail_CL
  }
  n26 = TextureSampleParameter2D() {
   group="##01.Textures"
   parameterName=Texture_BaseColor
   texture=/BRCosmetics/Athena/Effects/Particles/Skydiving/NS_Contrail_PajamaSoar/Textures/T_PajamaSoar_Contrail_CS
  }
  n27 = LinearInterpolate(A=@n25.RGB, B=@n26.RGB, Alpha=NamedRerouteUsage())
  n28 = ScalarParameter() {defaultValue=5, group=Toon_Shading_RimHighlighBacklit, parameterName=RimHighlight_Strength, sliderMax=10, sortPriority=16}
  n29 = LinearInterpolate(
   A=@n27
   B=LinearInterpolate(A=@n27, B=Multiply(A=@n26.RGB, B=@n28), Alpha=NamedRerouteUsage())
   Alpha=ScalarParameter() {group=Toon_Shading_RimHighlighBacklit, parameterName=UseRimHighlight, sliderMax=1, sortPriority=15}
  )
  n30 = StaticSwitchParameter(
   True=MaterialFunctionCall!(In=@n29, LUT, LUT Multiplier) {materialFunction=/Game/Packages/Fortress_SharedMaterials/Anime_Master_Material/MF_ApplyLUT}.Result
   False=@n29
  ) {group=Toon_Shading, parameterName=Use Apply LUT, sortPriority=0}
  n31 = Multiply(
   A=NamedRerouteUsage()
   B=Power(Base=NamedRerouteUsage(), Exp=ScalarParameter() {defaultValue=0.3, group=02.Metallic_Advance, parameterName=Metallic_LightingPower}) {constExponent=0.5}
  )
  n32 = NamedRerouteUsage()
  n33 = Multiply(
   A=Multiply(A=Multiply(A=@n23, B=@n24.RGB), B=LinearInterpolate(A=@n30, B=@n31, Alpha=@n32))
   B=StaticSwitchParameter(True=NamedRerouteUsage(), False=Constant() {r=1}) {parameterName=Enable_RimRouge}
  )
  n34 = Saturate(None=@n33)
  n35 = NamedRerouteUsage()
  n36 = NamedRerouteDeclaration(None=StaticSwitchParameter(True=LinearInterpolate(A=@n20, B=@n34, Alpha=@n35), False=@n34) {parameterName=Enable_InkLine}) {name=BaseColor_OUT, nodeColor=<rgba 0 1 0.3906 1>}
  n37 = StaticSwitchParameter(True=LinearInterpolate(A=@n36, B=NamedRerouteUsage(), Alpha=NamedRerouteUsage()), False=@n36) {parameterName=Enable_Highlight}
  n38 = NamedRerouteUsage()
  n39 = NamedRerouteUsage()
  n40 = NamedRerouteUsage()
  n41 = NamedRerouteUsage()
  n42 = Add(
   A=StaticSwitchParameter(True=LinearInterpolate(A=@n37, B=@n38, Alpha=@n39), False=@n37) {parameterName=Enable_Reflection}
   B=StaticSwitchParameter(True=Multiply(A=@n40, B=@n41), False=Constant()) {parameterName=Enable_Rimlight}
  )
  n43 = MaterialFunctionCall!(Color=@n42) {materialFunction=/Game/Packages/Fortress_SharedMaterials/Anime_Master_Material/MF_TODColorAdjustment}
  n44 = ScalarParameter() {defaultValue=1, group="########BasicSetting", parameterName=EmissiveScale}
  n45 = TextureSampleParameter2D() {
   group="##01.Textures"
   parameterName=Texture_Normal
   samplerType=SAMPLERTYPE_Normal
   texture=/BRCosmetics/Athena/Effects/Particles/Skydiving/NS_Contrail_PajamaSoar/Textures/T_PajamaSoar_Contrail_N
  }
  n46 = ScalarParameter() {group="########BasicSetting", parameterName=NormalFlatness}
  n47 = ScalarParameter() {defaultValue=3.5, group=WPO, parameterName=WPO Amplitude}
  n48 = ScalarParameter() {defaultValue=2.5, group=WPO, parameterName=WPO Frequency}
  n49 = Multiply(A=@n48) {constB=1.6}
  n50 = Subtract(A=Multiply(A=LinearInterpolate(A=@n48, B=@n49, Alpha=@n5) {constA=1, constB=4}, B=@n4), B=Add(A=@n1.Time Offset, B=Multiply(A=@n10, B=@n6)))
  n51 = Sine(None=@n50)
  n52 = Cosine(None=@n50)
  n53 = Sine(None=@n50) {period=3}
  n54 = ScalarParameter() {defaultValue=0.5, group=WPO, parameterName=WPO Z Adjustment}
  n55 = Multiply(
   A=LinearInterpolate(A=@n47, B=Multiply(A=@n47) {constB=0.45}, Alpha=@n5) {constA=2, constB=1.5}
   B=AppendVector(A=AppendVector(A=@n51, B=@n52), B=Multiply(A=@n53, B=@n54) {constB=0.6})
  ) {constB=3}
  n56 = ScalarParameter() {defaultValue=5, group=WPO, parameterName=WPO Amplitude 2}
  n57 = Sine(None=@n14)
  n58 = Constant()
  n59 = Multiply(A=Multiply(A=@n56) {constB=2}, B=AppendVector(A=AppendVector(A=@n57, B=@n58), B=@n58)) {constB=3}
  n60 = LinearInterpolate(A=@n56, B=Multiply(A=@n56), Alpha=@n5) {constA=2, constB=1.5}
  n61 = Constant()
  n62 = AppendVector(A=AppendVector(A=Sine(None=@n11), B=@n61), B=@n61)
  n63 = Multiply(A=@n4, B=Add(A=Add(A=@n55, B=@n59), B=Multiply(A=@n60, B=@n62) {constB=3}))
  n64 = ScalarParameter() {defaultValue=0.003, group=DontModify, parameterName=Subsurface Mask Value, sliderMax=1}
  n65 = NamedRerouteUsage()
  n66 = Constant()
  MaterialFunctionCall!(In=MakeMaterialAttributes!(
    BaseColor=@n18
    Metallic=StaticSwitchParameter(True=@n19.R, False=@n18) {parameterName=Enable_AdvanceTextureSetting}
    Specular=@n18
    Roughness=StaticSwitchParameter(True=@n19.G, False=@n18) {parameterName=Enable_AdvanceTextureSetting}
    Anisotropy
    EmissiveColor=Multiply(A=@n43.Result, B=@n44)
    Opacity=Constant() {r=1}
    OpacityMask
    Normal=MaterialFunctionCall(Normal=@n45.RGB, Flatness=@n46) {materialFunction=&FlattenNormal}.Result
    Tangent
    WorldPositionOffset=Transform(None=@n63) {transformSourceType=TRANSFORMSOURCE_Instance}
    SubsurfaceColor=MaterialFunctionCall(Y=@n64, Z=@n64) {materialFunction=&MakeFloat3}.Result
    ClearCoat
    ClearCoatRoughness
    AmbientOcclusion=StaticSwitchParameter(True=@n65, False=@n66) {parameterName=Enable_InkLine}
    Refraction
    CustomizedUVs_0
    CustomizedUVs_1
    CustomizedUVs_2
    CustomizedUVs_3
    CustomizedUVs_4
    CustomizedUVs_5
    CustomizedUVs_6
    CustomizedUVs_7
    PixelDepthOffset
    ShadingModel
    Displacement
   )) {materialFunction=/Game/Effects/Niagara/MasterMaterials/Functions/MF_FX_MAEmissiveAdjustments}
  MaterialFunctionCall() {materialFunction=&MF_ApplyLUT}
  Multiply(A=@n33, B=StaticSwitchParameter(True=NamedRerouteUsage(), False=Constant() {r=1}) {parameterName=Enable_SphereLighting})
  Multiply(A=@n26.RGB, B=NamedRerouteUsage()) {constA=1}
  NamedRerouteDeclaration(None=SkyAtmosphereLightDirection()) {name=SceneData_LightDir, nodeColor=<rgba 0 1 0.1797 1>}
  NamedRerouteDeclaration(None=CameraVectorWS()) {name=SceneData_CameraVector, nodeColor=<rgba 0.9219 1 0 1>}
  n67 = SkyLightEnvMapSample(Direction=ReflectionVectorWS() {bNormalizeCustomWorldNormal=true}, Roughness=NamedRerouteUsage())
  n68 = Desaturation(None=@n67, Fraction=Constant() {r=1})
  n69 = Normalize(VectorInput=VectorParameter() {group=02.Metallic_Advance, parameterName=Metallic_Reflection_Color, defaultValue=<rgba 1 1 1 1>}.RGB)
  n70 = LinearInterpolate(A=@n68, B=Multiply(A=@n69, B=@n68), Alpha=Saturate(None=DotProduct(A=@n67, B=@n69)))
  n71 = ScalarParameter() {defaultValue=10, group=02.Metallic_Advance, parameterName=Metallic_Step}
  n72 = Multiply(A=MaterialFunctionCall(Vector 3=@n70) {materialFunction=/Engine/Functions/Engine_MaterialFunctions02/Utility/VectorLength}.V3 Length, B=@n71) {constB=5}
  n73 = Multiply(A=Normalize(VectorInput=@n70), B=Divide(A=Ceil(None=@n72), B=@n71) {constB=5})
  n74 = Power(
   Base=NamedRerouteUsage()
   Exp=ScalarParameter() {defaultValue=30, group=02.Metallic_Advance, parameterName=Metallic_Painting_Str, sortPriority=64}
  ) {constExponent=50}
  NamedRerouteDeclaration(None=StaticSwitchParameter(True=LinearInterpolate(A=@n70, B=@n73, Alpha=@n74), False=@n73) {parameterName=Enable_PaintingEffect}) {name=Metallic_ReflectionColor, nodeColor=<rgba 0 1 0.3203 1>}
  n75 = WorldPosition()
  n76 = LinearInterpolate(
   A=ObjectPositionWS()
   B=ParticlePositionWS()
   Alpha=ScalarParameter() {defaultValue=1, group="########BasicSetting", parameterName=UseParticlePosition}
  )
  NamedRerouteDeclaration(None=Normalize(VectorInput=Subtract(A=@n75.XYZ, B=@n76))) {name=SceneData_SphereNormal, nodeColor=<rgba 0 0.7656 1 1>}
  n77 = Add(A=DotProduct(A=NamedRerouteUsage(), B=NamedRerouteUsage()))
  n78 = Saturate(None=OneMinus(None=Multiply(A=@n77) {constB=0.5}))
  n79 = ScalarParameter() {defaultValue=1, group="##11.SphereLighting", parameterName=SphereLighting_Power}
  NamedRerouteDeclaration(None=LinearInterpolate(
    B=VectorParameter() {group="##11.SphereLighting", parameterName=SphereLighting_Color, defaultValue=<rgba 2 2 2 1>}.RGB
    Alpha=Power(Base=@n78, Exp=@n79)
   ) {constA=1}) {name=SphereLighting_Color, nodeColor=<rgba 0 0.0156 1 1>}
  n80 = Add(A=NamedRerouteUsage(), B=NamedRerouteUsage())
  n81 = Transform(None=Constant3Vector() {constant=<rgba 1 0 0 0>}) {transformSourceType=TRANSFORMSOURCE_Camera}
  n82 = DotProduct(A=NamedRerouteUsage(), B=Normalize(VectorInput=Add(A=@n80, B=@n81)))
  n83 = Constant() {r=0.55}
  n84 = OneMinus(None=Saturate(None=Add(A=@n82, B=@n83)))
  n85 = NamedRerouteUsage()
  n86 = Fresnel(ExponentIn=ScalarParameter() {defaultValue=1.35, group="##21.RimRouge", parameterName=RimRouge_Fresnel}, Normal=@n85) {baseReflectFraction=0}
  NamedRerouteDeclaration(None=LinearInterpolate(
    B=VectorParameter() {group="##21.RimRouge", parameterName=RimRouge_Color, defaultValue=<rgba 2 1.55 1.4 1>}.RGB
    Alpha=Multiply(A=@n84, B=@n86)
   ) {constA=1}) {name=RimRouge_Color, nodeColor=<rgba 0.0547 1 0 1>}
  NamedRerouteDeclaration(None=VectorParameter() {group=10.Lighting_Advance, parameterName=Lighting_Color_2_LitRouge, defaultValue=<rgba 1 0.85 0.55 1>}.RGB) {name=Lighting_LitRougeColor, nodeColor=<rgba 0 1 0.1797 1>}
  NamedRerouteDeclaration(None=VectorParameter() {group=10.Lighting_Advance, parameterName=Lighting_Color_1_Lit, defaultValue=<rgba 1 1 1 1>}.RGB) {name=Lighting_LitColor, nodeColor=<rgba 0 0.0156 1 1>}
  n87 = VectorParameter() {group=10.Lighting_Advance, parameterName=Lighting_Color_3_ShadowRouge, defaultValue=<rgba 0.55 0.3644 0.3025 1>}
  NamedRerouteDeclaration(None=@n87.RGB) {name=Lighting_ShadowRougeColor, nodeColor=<rgba 0.7812 1 0 1>}
  NamedRerouteDeclaration(None=VectorParameter() {group=10.Lighting_Advance, parameterName=Lighting_Color_4_Shadow, defaultValue=<rgba 0.4 0.34 0.34 1>}.RGB) {name=Lighting_ShadowColor, nodeColor=<rgba 1 0 0.4453 1>}
  NamedRerouteDeclaration(None=VectorParameter() {group=10.Lighting_Advance, parameterName=Lighting_Color_5_FakeGI, defaultValue=<rgba 0.5 0.45 0.45 1>}.RGB) {name=Lighting_FakeGIColor, nodeColor=<rgba 0.0781 1 0 1>}
  n88 = NamedRerouteUsage()
  n89 = Saturate(None=@n88)
  n90 = Saturate(None=Multiply(A=@n88) {constB=-1})
  n91 = ScalarParameter() {defaultValue=0.1, group=10.Lighting_Advance, parameterName=Lighting_BlendWidth_RougeWithRouge}
  n92 = NamedRerouteUsage()
  n93 = StaticSwitchParameter(True=NamedRerouteUsage(), False=@n92) {parameterName=Enable_PaintingEffect}
  n94 = Multiply(A=@n91, B=@n93)
  n95 = Multiply(A=@n90, B=Divide(B=@n94))
  n96 = Saturate(None=Subtract(A=@n90, B=@n94))
  n97 = ScalarParameter() {defaultValue=0.2, group=10.Lighting_Advance, parameterName=Lighting_BlendWidth_ShadowRouge}
  n98 = Multiply(A=@n96, B=Divide(B=Multiply(A=@n97, B=@n93)))
  n99 = LinearInterpolate(
   A=LinearInterpolate(A=NamedRerouteUsage(), B=NamedRerouteUsage(), Alpha=Saturate(None=@n95))
   B=LinearInterpolate(A=NamedRerouteUsage(), B=NamedRerouteUsage(), Alpha=Saturate(None=@n98))
   Alpha=Ceil(None=@n96)
  )
  n100 = NamedRerouteUsage()
  n101 = Subtract(A=@n90, B=Multiply(A=@n91, B=@n92))
  n102 = Multiply(A=@n97, B=NamedRerouteUsage())
  n103 = Power(
   Base=Saturate(None=Subtract(A=@n101, B=@n102))
   Exp=ScalarParameter() {defaultValue=1.35, group=10.Lighting_Advance, parameterName=Lighting_FakeGIAreaPower}
  )
  n104 = ScalarParameter() {defaultValue=3, group=10.Lighting_Advance, parameterName=Lighting_GIStep}
  n105 = Divide(A=Floor(None=Multiply(A=@n103, B=@n104) {constB=10}), B=@n104) {constB=10}
  n106 = MaterialFunctionCall!(
   HSV=VectorParameter() {group="##10.Lighting", parameterName=Lighting_FinalColorHSV, defaultValue=<rgba 0 1 1 1>}.RGB
   Color=LinearInterpolate(A=LinearInterpolate(A=@n99, B=@n100, Alpha=@n105), B=NamedRerouteUsage(), Alpha=Ceil(None=@n89))
  ) {materialFunction=/Game/Effects/Niagara/MasterMaterials/Functions/Common/FXCommon_HSV}
  n107 = NamedRerouteUsage()
  n108 = Multiply(
   A=ScalarParameter() {defaultValue=0.1, group=10.Lighting_Advance, parameterName=Lighting_BlendWidth_LitRouge}
   B=StaticSwitchParameter(True=NamedRerouteUsage(), False=NamedRerouteUsage()) {parameterName=Enable_PaintingEffect}
  )
  n109 = Saturate(None=Multiply(A=@n89, B=Divide(B=@n108)))
  n110 = Multiply(
   A=VectorParameter() {group="##10.Lighting", parameterName=Lighting_Simple_LitColor, defaultValue=<rgba 1 1 1 1>}.RGB
   B=ScalarParameter() {defaultValue=1, group="##10.Lighting", parameterName=Lighting_Simple_LitStr}
  )
  n111 = Multiply(
   A=VectorParameter() {group="##10.Lighting", parameterName=Lighting_Simple_ShadowColor, defaultValue=<rgba 1 1 1 1>}.RGB
   B=ScalarParameter() {defaultValue=0.5, group="##10.Lighting", parameterName=Lighting_Simple_ShadowStr}
  )
  n112 = ScalarParameter() {defaultValue=4, group="##10.Lighting", parameterName=Lighting_Hardness}
  n113 = SmoothStep(Max=Divide(B=@n112) {constA=0.08}, Value=@n90) {constMax=0.02}
  NamedRerouteDeclaration(None=StaticSwitchParameter(True=LinearInterpolate(A=@n106.Result, B=@n107, Alpha=@n109), False=LinearInterpolate(A=@n110, B=@n111, Alpha=@n113) {constA=1, constB=0.5}) {parameterName=Enable_Lighting_ComplexBlending}) {name=Lighting_Color, nodeColor=<rgba 0.4297 1 0 1>}
  n114 = ScalarParameter() {defaultValue=0.2, group=10.Lighting_Advance, parameterName=Lighting_Painting_Hardness, sortPriority=64}
  n115 = NamedRerouteUsage()
  n116 = Constant() {r=0.2}
  n117 = Multiply(A=@n112, B=Add(A=Multiply(A=@n115, B=@n116)))
  n118 = Constant() {r=0.1}
  n119 = Power(
   Base=NamedRerouteUsage()
   Exp=LinearInterpolate(
    A=ScalarParameter() {defaultValue=50, group=10.Lighting_Advance, parameterName=Lighting_Painting_ShadowFaceStrength, sortPriority=64}
    B=ScalarParameter() {group=10.Lighting_Advance, parameterName=Lighting_Painting_LitFaceStrength, sortPriority=64}
    Alpha=MaterialFunctionCall!(In=@n88, Ramp Width=@n118) {materialFunction=/Engine/Functions/Engine_MaterialFunctions01/Math/SmoothCeil}.Result
   )
  )
  NamedRerouteDeclaration(None=Divide(B=LinearInterpolate(A=@n114, B=@n117, Alpha=@n119))) {name=Lighting_BlendWidthScaleWithPainting, nodeColor=<rgba 1 0.2812 0 1>}
  NamedRerouteDeclaration(None=Divide(B=@n117)) {name=Lighting_BlendWidthScale, nodeColor=<rgba 0.8984 1 0 1>}
  n120 = TextureSampleParameter2D() {
   group="##40.InkLine"
   parameterName=Texture_InkLines
   samplerSource=SSM_Wrap_WorldGroupSettings
   samplerType=SAMPLERTYPE_Masks
   sortPriority=1
   texture=&T_BlackMask
  }
  n121 = Max(B=@n120.R) {desc=Inject Inklines here}
  n122 = ScalarParameter() {defaultValue=15, group=40.InkLine_Advance, parameterName=Inkline_BlurPow, sliderMax=16}
  n123 = OneMinus(None=Multiply(
    A=Power(Base=@n121, Exp=@n122) {constExponent=3}
    B=ScalarParameter() {defaultValue=0.5, group=40.InkLine_Advance, parameterName=Inkline_BlurMult, sliderMax=2}
   ) {constB=0.5})
  n124 = Constant() {r=1}
  n125 = ScalarParameter() {defaultValue=100, group=40.InkLine_Advance, parameterName=Inkline_FarDistance, sliderMax=1000}
  n126 = Subtract(A=Distance(A=WorldPosition().XYZ, B=CameraPositionWS()), B=@n125)
  n127 = Subtract(A=ScalarParameter() {defaultValue=500, group=40.InkLine_Advance, parameterName=Inkline_NearDistance, sliderMax=1000}, B=@n125)
  n128 = LinearInterpolate(
   A=ScalarParameter() {defaultValue=0.02, group=40.InkLine_Advance, parameterName=Inkline_LineThickness_Near, sliderMax=1}
   B=ScalarParameter() {defaultValue=0.1, group=40.InkLine_Advance, parameterName=Inkline_LineThickness_Far, sliderMax=1}
   Alpha=Saturate(None=Divide(A=@n126, B=@n127))
  )
  n129 = ScalarParameter() {defaultValue=1, group=40.InkLine_Advance, parameterName=Inkline_LineHardness, sliderMax=1}
  n130 = OneMinus(None=Saturate(None=SphereMask(A=@n121, B=@n124, Radius=@n128, Hardness=@n129) {attenuationRadius=0.1}))
  NamedRerouteDeclaration(None=Saturate(None=Multiply(A=@n123, B=@n130))) {name=InkLine Mask, nodeColor=<rgba 0.1615 0.1615 0.1615 1>}
  n131 = TextureSampleParameter2D() {
   group=Toon_Shading
   parameterName=Texture_DF_Color
   samplerType=SAMPLERTYPE_Alpha
   texture=/BRCosmetics/Athena/Effects/Particles/Skydiving/NS_Contrail_PajamaSoar/Textures/T_PajamaSoar_Contrail_DFL
  }
  NamedRerouteDeclaration(None=StaticSwitchParameter(True=@n131.RGB, False=VectorParameter() {group="##40.InkLine", parameterName=InkLineColor, defaultValue=<rgba 0 0 0 1>}.RGB) {group=Toon_Shading, parameterName=UseDFColor}) {name=InkLine Color, nodeColor=<rgba 0.1615 0.1615 0.1615 1>}
  NamedRerouteDeclaration(None=ScalarParameter() {defaultValue=0.2, group="########BasicSetting", parameterName=Roughness}) {name=BaseParam_Roughness, nodeColor=<rgba 1 0 0.5625 1>}
  n132 = TextureSampleParameter2D() {group="##01.Textures", parameterName=Texture_MHS, samplerType=SAMPLERTYPE_Masks, texture=&T_BlackMask}
  n133 = VectorParameter() {group="########BasicSetting", parameterName=MHSTextureScale, sortPriority=64, defaultValue=<rgba 1 1 1 1>}
  n134 = MaterialFunctionCall(Float3=StaticSwitchParameter(True=Multiply(A=@n132.RGB, B=@n133.RGB), False=Constant()) {parameterName=Enable_AdvanceTextureSetting}) {materialFunction=&BreakOutFloat3Components}
  n135 = ScalarParameter() {group="########BasicSetting", parameterName=Metallic}
  NamedRerouteDeclaration(None=Saturate(None=Add(A=@n134.R, B=@n135))) {name=BaseParam_Metallic, nodeColor=<rgba 1 0 0.5625 1>}
  n136 = ScalarParameter() {defaultValue=1, group="########BasicSetting", parameterName=Specular}
  NamedRerouteDeclaration(None=Saturate(None=Add(A=@n134.B, B=@n136))) {name=BaseParam_Specular, nodeColor=<rgba 1 0 0.5625 1>}
  n137 = ScalarParameter() {defaultValue=0.3, group="########BasicSetting", parameterName=HighLight}
  NamedRerouteDeclaration(None=Saturate(None=Add(A=@n134.G, B=@n137))) {name=BaseParam_HighLight, nodeColor=<rgba 1 0 0.5625 1>}
  n138 = NamedRerouteUsage()
  n139 = DDX(Value=@n138)
  n140 = DDY(Value=@n138)
  n141 = ScalarParameter() {defaultValue=1500, group="##41.PaintingEffect", parameterName=PaintingEffect_AreaExpand}
  n142 = Distance(A=CameraPositionWS(), B=WorldPosition().XYZ)
  n143 = Divide(A=Multiply(A=@n141, B=@n141), B=Multiply(A=@n142, B=@n142))
  n144 = VectorParameter() {
   group="##41.PaintingEffect"
   parameterName=PaintingEffect_DistanceFade
   channelNames {a="", b="", g=Far, r=Near}
   defaultValue=<rgba 30 100 0 1>
  }
  n145 = Saturate(None=Divide(A=Subtract(A=@n142, B=@n144.Near) {constB=30}, B=@n144.Far) {constB=100})
  n146 = MaterialFunctionCall(Vector 3=Multiply(A=CrossProduct(A=@n139, B=@n140), B=LinearInterpolate(B=@n143, Alpha=@n145))) {materialFunction=&VectorLength}
  NamedRerouteDeclaration(None=OneMinus(None=Saturate(None=@n146.V3 Length))) {name=Painting_OriginalOutput, nodeColor=<rgba 0.7109 0 1 1>}
  n147 = ScalarParameter() {group=Toon_Shading_Hair, parameterName=IgnoreZforSpecularHighlight, sliderMax=1}
  n148 = StaticSwitchParameter(
   True=LinearInterpolate(A=ScalarParameter() {group=Toon_Shading_Hair, parameterName=IgnoreZforSpecularHighlight2, sliderMax=1}, B=@n147, Alpha=NamedRerouteUsage())
   False=@n147
  ) {
   desc=Use 2 IgnoreZforSpecularHighlight values based on preskin position.
   group=Toon_Shading_Hair
   parameterName=UseIgnoreZforSpecularHighlightBlend
   sortPriority=0
  }
  n149 = AppendVector(A=Constant2Vector() {g=1, r=1}, B=Saturate(None=OneMinus(None=@n148)))
  n150 = Multiply(A=NamedRerouteUsage(), B=@n149)
  n151 = Multiply(A=MaterialFunctionCall() {materialFunction=&CameraDirectionVector}.Result) {constB=-1}
  n152 = NamedRerouteUsage()
  n153 = Multiply(
   A=Add(
    A=CameraVectorWS()
    B=StaticSwitchParameter(True=@n151, False=@n152) {group=Toon_Shading_Hair, parameterName=UseCameraDirectionForSpecularHighlights, sortPriority=0}
   )
   B=@n149
  )
  n154 = Multiply(A=@n149, B=PixelNormalWS())
  n155 = StaticSwitchParameter(
   True=DotProduct(A=ReflectionVectorWS(), B=Normalize(VectorInput=@n150)) {desc=Phong}
   False=DotProduct(A=Normalize(VectorInput=@n153), B=Normalize(VectorInput=@n154)) {desc=Blinn}
  ) {desc=Switches between Blinn and Phong Highlight., group=Toon_Shading_Hair, parameterName=UsePhongHighlight, sortPriority=0}
  n156 = Power(Base=Saturate(None=@n155), Exp=ScalarParameter() {defaultValue=50, group=Toon_Shading_Hair, parameterName=SpecPower, sliderMax=32})
  n157 = ScalarParameter() {defaultValue=4, group=Toon_Shading_Hair, parameterName=Sharpness, sliderMax=8}
  NamedRerouteDeclaration(None=Multiply(
    A=MaterialFunctionCall(In=@n156, Contrast=@n157) {materialFunction=&CheapContrast}.Result
    B=NamedRerouteUsage()
   )) {name=SpecularHighlightMask, nodeColor=<rgba 0 1 0.25 1>}
  NamedRerouteDeclaration(None=PixelNormalWS()) {name=SceneData_PixelNormal, nodeColor=<rgba 0 0.625 1 1>}
  n158 = NamedRerouteUsage()
  n159 = VectorParameter() {group=Toon_Shading_Hair, parameterName=SpecularColor, defaultValue=<rgba 1 1 1 1>}
  NamedRerouteDeclaration(None=Multiply(
    A=Multiply(A=@n158, B=@n159.RGB)
    B=ScalarParameter() {defaultValue=10, group=Toon_Shading_Hair, parameterName=SpecularStrength, sliderMax=20}
   )) {name=SpecularColor, nodeColor=<rgba 0 0.8125 1 1>}
  n160 = PreSkinnedPosition()
  n161 = NamedRerouteDeclaration(None=Divide(A=VertexInterpolator(VS=@n160).PS, B=PreSkinnedLocalBounds().Max)) {name=PreskinLocPosNormalized, nodeColor=<rgba 1 0.9609 0 1>}
  n162 = Divide(
   A=Distance(
    A=ComponentMask(None=@n161) {b=true}
    B=ScalarParameter() {defaultValue=0.25, group=Toon_Shading_Hair, parameterName=IgnoreZforSpecularHighlight_BlendPosition}
   )
   B=ScalarParameter() {defaultValue=0.325, group=Toon_Shading_Hair, parameterName=IgnoreZforSpecularHighlight_BlendSpread}
  ) {constB=0.4}
  n163 = ScalarParameter() {defaultValue=4, group=Toon_Shading_Hair, parameterName=IgnoreZforSpecularHighlight_BlendContrast}
  NamedRerouteDeclaration(None=Saturate(None=MaterialFunctionCall(In=@n162, Contrast=@n163) {materialFunction=&CheapContrast}.Result)) {name=IgnoreZblend, nodeColor=<rgba 0 0.4141 1 1>}
  NamedRerouteDeclaration(None=@n19.B) {name=SSC_Blue, nodeColor=<rgba 0 0.055 1 1>}
  NamedRerouteDeclaration(None=ChannelMaskParameter(None=@n19.RGB) {group=Toon_Shading_SSC, parameterName=SSC_ThresholdChannel, sortPriority=10}.R) {name=SSC_ThresholdChannel, nodeColor=<rgba 0 0.9766 1 1>}
  n164 = NamedRerouteUsage()
  n165 = DotProduct(A=@n164, B=NamedRerouteUsage())
  n166 = Add(A=@n165, B=ScalarParameter() {group=Toon_Shading, parameterName=Lit Bias, sliderMax=1})
  n167 = MaterialFunctionCall(
   Alpha=Power(Base=NamedRerouteUsage(), Exp=ScalarParameter() {defaultValue=2, group=Toon_Shading, parameterName=ThresholdmaskPower, sliderMax=8})
   Min=ScalarParameter() {group=Toon_Shading, parameterName=Shaded_Threshold_Min, sliderMax=1}
   Max=ScalarParameter() {defaultValue=1, group=Toon_Shading, parameterName=Shaded_Threshold_Max, sliderMax=1}
  ) {materialFunction=&SmoothStep}
  n168 = ScalarParameter() {defaultValue=12, group=Toon_Shading_RimHighlighBacklit, parameterName=BackLitRim_OutlineC2, sliderMax=16, sliderMin=-8}
  n169 = Fresnel(
   ExponentIn=ScalarParameter() {defaultValue=2, group=Toon_Shading_RimHighlighBacklit, parameterName=BackLitExponent1, sliderMax=8}
   Normal=@n164
  ) {baseReflectFraction=0}
  n170 = LinearInterpolate(
   A=ScalarParameter() {defaultValue=-4, group=Toon_Shading_RimHighlighBacklit, parameterName=BackLitRim_OutlineC1, sliderMax=16}
   B=@n168
   Alpha=Power(
    Base=OneMinus(None=@n169)
    Exp=ScalarParameter() {defaultValue=4, group=Toon_Shading_RimHighlighBacklit, parameterName=BacklitOutlineExponent2, sliderMax=8}
   )
  )
  n171 = Add(A=OneMinus(None=Saturate(None=@n170)), B=Saturate(None=@n165))
  n172 = Add(
   A=DotProduct(
    A=MaterialFunctionCall() {materialFunction=&CameraDirectionVector}.Result
    B=NamedRerouteUsage()
   )
   B=ScalarParameter() {defaultValue=0.25, group=Toon_Shading_RimHighlighBacklit, parameterName=BackLitBias, sliderMax=1}
  )
  n173 = ScalarParameter() {defaultValue=0.75, group=Toon_Shading_RimHighlighBacklit, parameterName=BackLitMaxBias, sliderMax=1}
  n174 = CurveAtlasRowParameter(CurveTime=LinearInterpolate(A=Multiply(A=@n166, B=@n167.Result), B=Multiply(A=@n171, B=@n167.Result), Alpha=Clamp(None=@n172, Max=@n173))) {
   atlas=/Game/Characters/Shared/CurveAtlas/AthenaCharacterCurveAtlas
   curve=/Game/Packages/Fortress_SharedMaterials/Anime_Master_Material/TwoToneShadingCurve
   defaultValue=14
   group=Toon_Shading
   parameterName=ShadingCurve
   sortPriority=5
  }
  NamedRerouteDeclaration(None=ChannelMaskParameter(None=@n174) {group=Toon_Shading, parameterName=ShadingCurveChannel, sortPriority=10}.R) {name=ShadingCurve, nodeColor=<rgba 0 1 0.625 1>}
  n175 = NamedRerouteUsage()
  n176 = PixelNormalWS()
  n177 = Fresnel(
   ExponentIn=ScalarParameter() {defaultValue=4, group=Toon_Shading_RimHighlighBacklit, parameterName=BL_RimHighlightExponent1, sliderMax=8}
   Normal=PixelNormalWS()
  ) {baseReflectFraction=0}
  n178 = LinearInterpolate(
   A=ScalarParameter() {defaultValue=-4, group=Toon_Shading_RimHighlighBacklit, parameterName=BL_Rim_OutlineC1, sliderMax=16, sliderMin=-8}
   B=ScalarParameter() {defaultValue=12, group=Toon_Shading_RimHighlighBacklit, parameterName=BL_Rim_OutlineC2, sliderMax=16, sliderMin=-8}
   Alpha=Power(
    Base=OneMinus(None=@n177)
    Exp=ScalarParameter() {defaultValue=4, group=Toon_Shading_RimHighlighBacklit, parameterName=BL_Rim_OutlineExponent2, sliderMax=8}
   )
  )
  n179 = Multiply(A=Saturate(None=DotProduct(A=@n175, B=@n176)), B=OneMinus(None=Saturate(None=@n178)))
  n180 = NamedRerouteUsage()
  n181 = StaticSwitchParameter(
   True=Multiply(A=@n179, B=ChannelMaskParameter(None=@n180) {group=Toon_Shading_RimHighlighBacklit, parameterName=Rim_FXMask_Channel, sortPriority=10}.R)
   False=@n179
  ) {group=Toon_Shading_RimHighlighBacklit, parameterName=UseFXMaskforRimHighlight, sortPriority=0}
  n182 = VectorParameter() {
   desc=CameraVectortoReduceRimHighlight
   group=Toon_Shading_RimHighlighBacklit
   parameterName=RimHighlight_CustomTopVector
   defaultValue=<rgba 0 0 1 1>
  }
  n183 = ComponentMask(None=Transform(None=Normalize(VectorInput=@n182.RGB)) {transformSourceType=TRANSFORMSOURCE_Camera, transformType=TRANSFORM_Local}) {b=true}
  n184 = Power(Base=Saturate(None=Multiply(A=@n183) {constB=-1})) {constExponent=10}
  n185 = ScalarParameter() {defaultValue=1, group=Toon_Shading_RimHighlighBacklit, parameterName=RimHighlight_TopReducing, sliderMax=10}
  n186 = Saturate(None=Add(A=OneMinus(None=@n184), B=OneMinus(None=@n185)))
  NamedRerouteDeclaration(None=StaticSwitchParameter(True=Multiply(A=@n181, B=@n186), False=@n181) {group=Toon_Shading_RimHighlighBacklit, parameterName=UseReduceTopRimHighlight, sortPriority=0}) {name=Rim Highlight, nodeColor=<rgba 1 0 0.1172 1>}
  NamedRerouteDeclaration(None=Round(None=@n167.Result)) {name=SSC Threshold, nodeColor=<rgba 0.7973 1 0 1>}
  n187 = ScalarParameter() {defaultValue=0.005, group="##31.Highlight", parameterName=HighLight_Softness}
  n188 = NamedRerouteUsage()
  n189 = AppendVector(A=Constant2Vector() {g=1, r=1}, B=OneMinus(None=ScalarParameter() {group="##31.Highlight", parameterName=HighLight_IgnoreZ}))
  n190 = NamedRerouteUsage()
  n191 = NamedRerouteUsage()
  n192 = MaterialFunctionCall!(
   LightDirctionIn=Normalize(VectorInput=Multiply(A=@n188, B=@n189))
   CameraVectorIn=Normalize(VectorInput=Multiply(A=@n190, B=@n189))
   Roughness=NamedRerouteUsage()
   NormalIn=Normalize(VectorInput=Multiply(A=@n191, B=@n189))
  ) {materialFunction=/Game/Effects/Niagara/MasterMaterials/Functions/Common/FXCommon_GGX}
  n193 = Saturate(None=SmoothStep(Max=Add(A=@n187) {constB=0.05}, Value=Saturate(None=@n192.Result)) {constMin=0.05})
  NamedRerouteDeclaration(None=@n193) {name=GGX_Mask, nodeColor=<rgba 0.1484 0 1 1>}
  NamedRerouteDeclaration(None=LinearInterpolate(
    A=VectorParameter() {group="##31.Highlight", parameterName=HighLight_Rouge, defaultValue=<rgba 1 1 1 1>}.RGB
    B=VectorParameter() {group="##31.Highlight", parameterName=HighLight_Color, defaultValue=<rgba 1 1 1 1>}.RGB
    Alpha=@n193
   )) {name=GGX_Color, nodeColor=<rgba 0.1719 0 1 1>}
  n194 = NamedRerouteUsage()
  n195 = ScalarParameter() {defaultValue=5, group="##30.Reflection", parameterName=Reflection_Step}
  n196 = Multiply(
   A=Fresnel(
    ExponentIn=ScalarParameter() {defaultValue=0.9, group="##30.Reflection", parameterName=Reflection_Painting_Fresnel, sortPriority=64}
    Normal=@n194
   ) {baseReflectFraction=0}
   B=@n195
  ) {constB=5}
  n197 = Fresnel(ExponentIn=ScalarParameter() {defaultValue=1.35, group="##30.Reflection", parameterName=Reflection_Fresnel}, Normal=@n194) {baseReflectFraction=0}
  n198 = Divide(A=Floor(None=Multiply(A=@n197, B=@n195) {constB=5}), B=@n195) {constB=5}
  n199 = LinearInterpolate(
   A=Divide(A=Floor(None=@n196), B=@n195) {constB=5}
   B=@n198
   Alpha=Power(
    Base=NamedRerouteUsage()
    Exp=ScalarParameter() {defaultValue=50, group="##30.Reflection", parameterName=Reflection_Painting_Str, sortPriority=64}
   ) {constExponent=50}
  )
  NamedRerouteDeclaration(None=Multiply(
    A=StaticSwitchParameter(True=@n199, False=@n198) {parameterName=Enable_PaintingEffect}
    B=ScalarParameter() {defaultValue=0.5, group="##30.Reflection", parameterName=Reflection_Strength}
   )) {name=Reflection_Mask, nodeColor=<rgba 1 0 0.5156 1>}
  n200 = Normalize(VectorInput=@n67)
  n201 = MaterialFunctionCall(Vector 3=@n67) {materialFunction=&VectorLength}
  n202 = Divide(A=Ceil(None=Multiply(A=@n201.V3 Length, B=@n195) {constB=5}), B=@n195) {constB=5}
  NamedRerouteDeclaration(None=Desaturation(None=Multiply(A=@n200, B=@n202), Fraction=ScalarParameter() {defaultValue=-1, group="##30.Reflection", parameterName=Reflection_Desaturation})) {name=Reflection_Color, nodeColor=<rgba 1 0 0.9375 1>}
  n203 = NamedRerouteUsage()
  n204 = NamedRerouteUsage()
  n205 = Power(Base=Saturate(None=DotProduct(A=@n203, B=@n204)))
  n206 = OneMinus(None=ScalarParameter() {defaultValue=0.3, group="##20.RimLight", parameterName=RimLight_Width})
  n207 = Add(A=ScalarParameter() {defaultValue=0.02, group="##20.RimLight", parameterName=RimLight_Softness}, B=@n206) {constB=0.5}
  n208 = DotProduct(A=PixelNormalWS(), B=CameraVectorWS())
  n209 = Saturate(None=SmoothStep(Min=@n206, Max=Saturate(None=@n207), Value=OneMinus(None=@n208)) {constMax=0.51, constMin=0.5})
  NamedRerouteDeclaration(None=Saturate(None=Multiply(A=@n205, B=@n209))) {name=RimLight_Mask, nodeColor=<rgba 1 0 0.2109 1>}
  NamedRerouteDeclaration(None=VectorParameter() {group="##20.RimLight", parameterName=RimLight_Color, defaultValue=<rgba 2 2 2 1>}.RGB) {name=RimLight_Color, nodeColor=<rgba 1 0 0.2109 1>}
  NamedRerouteUsage()
  NamedRerouteUsage()
  n210 = NamedRerouteUsage()
  Normalize(VectorInput=LinearInterpolate(A=@n85, B=Add(A=@n210, B=@n85), Alpha=Constant()))
  ScalarParameter() {defaultValue=2, group=WPO, parameterName=WPO Frequency 2}
  ScalarParameter() {defaultValue=2, group=WPO, parameterName=WPO Frequency 2}
  StaticSwitchParameter(
   True=LinearInterpolate(
    A=@n42
    B=MaterialFunctionCall!(Color=@n42) {materialFunction=/Game/Packages/Fortress_SharedMaterials/Anime_Master_Material/MF_TODColorCurve}.Result
    Alpha=ScalarParameter() {defaultValue=1, group=TOD_Adjustments, parameterName=TODStr}
   )
   False=@n42
  ) {defaultValue=true, parameterName=Enable_TOD}
  n211 = Add(B=LinearInterpolate(A=ScalarParameter() {defaultValue=-0.4, group="##10.Lighting", parameterName=Lighting_Offset}, Alpha=NamedRerouteUsage()) {constB=0})
  n212 = OneMinus(None=Constant() {r=0.6})
  n213 = Constant() {r=0.01}
  n214 = PixelNormalWS()
  n215 = Saturate(None=SmoothStep(Min=@n212, Max=Add(A=@n212, B=@n213), Value=Fresnel(Normal=@n214) {baseReflectFraction=0, exponent=1}))
  n216 = Add(A=DotProduct(
    A=MaterialFunctionCall() {materialFunction=&CameraDirectionVector}.Result
    B=NamedRerouteUsage()
   )) {constB=0.25}
  StaticSwitchParameter(True=LinearInterpolate(A=@n211, B=Add(A=@n215, B=@n211), Alpha=Clamp(None=@n216) {maxDefault=0.75}), False=@n211) {parameterName=Enable_BackfaceLit}
  Time()
 ]
}
