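# Skydiving contrail system that attaches a pair of balloons, hand wraps, wires and confetti to the player's skeletal mesh.
# Balloons and wraps are persistent burst-spawned meshes re-sampled from bones each update, with rotation blended by dive value and lean alpha feeding balloon materials.
# Wires are chained ribbon particles displaced by curves, velocity direction and animated sine sway.
# Confetti sprites spawn faster when diving, drift under curl noise and fade by camera distance.
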NiagaraSystem /BRCosmetics/Athena/Effects/Particles/Skydiving/NS_Contrail_PajamaSoar/FX/NS_Contrail_PajamaSoar
dependencies {
 usedDataInterfaces [
  &NiagaraDataInterfaceArrayFloat
  &NiagaraDataInterfaceArrayFloat3
  &NiagaraDataInterfaceVectorCurve
  &NiagaraDataInterfaceCurve
  &NiagaraDataInterfaceCamera
  &NiagaraDataInterfaceSpriteRendererInfo
 ]
 usedDynamicInputs [
  &SetBoolByFloatComparison
  &RotateVectorByQuaternion
  /Niagara/DynamicInputs/Rotation/InvertQuaternion
 ]
 usedModules [
  &self:SystemSpawnScript.SystemScriptSource.SystemScriptGraph.NiagaraNodeAssignment_0.SetVariables_#1
  &SystemState
  /Game/Effects/Niagara/Modules/Contrails/FN_Contrail_Setup
  &FN_Integrator
  /Game/Effects/Niagara/Modules/Skydiving/FN_DivingAttributes/v4/FN_DivingAttributes_v4
  &self:SystemSpawnScript.SystemScriptSource.SystemScriptGraph.NiagaraNodeAssignment_1.SetVariables_#2
  &self:SystemSpawnScript.SystemScriptSource.SystemScriptGraph.NiagaraNodeAssignment_2.SetVariables_#3
 ]
}
emitters {
 BallonWire_Hand {
  emitterSpawnScript [&FN_PreserveSkydiveUserVariables() {module=PreserveUserVariables}]
  emitterUpdateScript [
   &EmitterState(
    Inactive Response=::NewEnumerator1
    Life Cycle Mode=::NewEnumerator1
    Loop Behavior=::NewEnumerator1
    Loop Duration Mode=::NewEnumerator1
    UseLoopDelay=false
   )
   /SetVariables(Emitter.Number of Chains=8, Emitter.Particle per Chain=5) {isSetParameters=true}
   &SpawnBurst_Instantaneous(
    Spawn Count=&Multiply_Int(A=$Emitter.Number of Chains, B=$Emitter.Particle per Chain)
    Spawn Probability=HlslExpression() {hlslExpression="Emitter.LoopCount == 0"}
    Use Spawn Probability=false
   ) {module=SpawnBurst_Instantaneous001}
  ]
  particleSpawnScript [
   &FN_SetSpawnGroup(
    Inclusivity=true
    Normalize Spawn Group=true
    Particles Per Group:NiagaraInt32=$Emitter.Particle per Chain
    Pattern=::NewEnumerator0
    Set Ribbon ID and Ribbon Link Order=true
   )
   &InitializeParticle(Color=<rgba 1.1409 1.2 1.2 1>, Ribbon Width=0.85, Ribbon Width Mode=::NewEnumerator1, UsePositionOffset=false, Write Material Random=true) {module=InitializeParticle001}
  ]
  particleUpdateScript [
   &ParticleState(Kill Particles When Lifetime Has Elapsed=false)
   &SkeletalMeshLocation(
    Bone / Socket Index=&ModuloInt(A=&SpawnGroup(), B=2)
    Bone Sampling Mode=::NewEnumerator4
    MaskSpawnGroup=false
    Orientation Sampling=::NewEnumerator1
    Sampled Position Offset=&AddVectorToPosition(
     Position=<0 0 5>
     Vector=&TransformVector(
      Vector=&VectorFromCurve(
       VectorCurve=DataInterface() {propertyValues {
         XCurve {keys [
           {arriveTangent=4.1785, interpMode=RCIM_Cubic, leaveTangent=4.1785, tangentMode=RCTM_User, time=0, value=0}
           {arriveTangent=1.1179, interpMode=RCIM_Cubic, leaveTangent=1.1179, tangentMode=RCTM_User, time=0.2042, value=0.6228}
           {arriveTangent=0.0224, interpMode=RCIM_Cubic, leaveTangent=0.0224, tangentMode=RCTM_User, time=1, value=1}
          ]}
         YCurve {keys [
           {arriveTangent=4.1785, interpMode=RCIM_Cubic, leaveTangent=4.1785, tangentMode=RCTM_User, time=0, value=0}
           {arriveTangent=1.1179, interpMode=RCIM_Cubic, leaveTangent=1.1179, tangentMode=RCTM_User, time=0.2042, value=0.6228}
           {arriveTangent=0.0224, interpMode=RCIM_Cubic, leaveTangent=0.0224, tangentMode=RCTM_User, time=1, value=1}
          ]}
         ZCurve {keys [
           {arriveTangent=4.1785, interpMode=RCIM_Cubic, leaveTangent=4.1785, tangentMode=RCTM_User, time=0, value=0}
           {arriveTangent=1.1179, interpMode=RCIM_Cubic, leaveTangent=1.1179, tangentMode=RCTM_User, time=0.2042, value=0.6228}
           {arriveTangent=0.0224, interpMode=RCIM_Cubic, leaveTangent=0.0224, tangentMode=RCTM_User, time=1, value=1}
          ]}
        }}
       CurveIndex=$Particles.RibbonLinkOrder
       Scale Curve=<3 5 5>
      )
      Destination Space=::Local
     )
    )
    Sampled Position Offset Coordinate Space:ENiagaraSkelMeshTransforms=::NewEnumerator2
    Skeletal Mesh=$System.Skeletal Mesh
    Use Sampled Mesh Attribute as Ribbon ID=false
   )
   /SetVariables(Particles.Position=&AddVectorToPosition(
     Position=$Particles.Position
     Vector=&Multiply_VectorByFloat(
      Vector=&Add_Vector(
       A=&Multiply_Vector(
        A=&NormalizeVector(
         Vector To Normalize=&FN_FrontEndVector3Modifier(Gameplay=$System.PlayerVelDir, Front End=$System.FrontendVelocity)
         Safe Normalize=true
         Safe Fallback Threshold=0
         Safe Fallback Vector=<0 0 0.001>
        )
        B=<0 0 -1>
       )
       B=&Multiply_VectorByFloat(
        Vector=&TransformVector(
         Vector=&MergeHelper_VectorSwitch(
          0=<1 0 0>
          1=<-1 0 0>
          2=<1 0 0>
          3=<-1 0 0>
          4=<1 0 0>
          5=<-1 0 0>
          6=<1 0 0>
          7=<-1 0 0>
          Index=$Particles.SpawnGroup
          Num=::NewEnumerator6
         )
         Destination Space=::Simulation
        )
        Float=1
       )
      )
      Float=&FloatFromCurve(
       FloatCurve=DataInterface() {propertyValues {Curve {defaultValue=0, keys [<at 0 0 RCIM_Cubic>, {arriveTangent=2.4094, tangentMode=RCTM_User, time=1, value=1}]}}}
       CurveIndex=$Particles.Normalized Spawn Group
       Scale Curve=&RandomRangeFloat(Minimum=10, Maximum=15, Random Seed=$Particles.SpawnGroup)
      )
     )
    )) {isSetParameters=true}
   /SetVariables(Particles.Position=&AddVectorToPosition(
     Position=$Particles.Position
     Vector=&Multiply_VectorByFloat(
      Vector=/Niagara/DynamicInputs/Vectors/RotateVector(
       Vector To Rotate:Vector3f=<10 0 3>
       Yaw:NiagaraFloat=&Add_Float(
        A=$Particles.Age
        B=&Multiply_Float(A=&MakeFloatFromInt(Int=&SpawnGroup()), B=3)
       )
       Pitch:NiagaraFloat=0
       Roll:NiagaraFloat=0
       Rotation Rate:NiagaraFloat=&Multiply_Float(A=1, B=200)
      )
      Float=&Multiply_Float(
       A=&Sine(
        Normalized Angle=&Add_Float(A=&Multiply_Float(A=$Emitter.Age, B=-3), B=$Particles.Normalized Spawn Group)
        Period=2
        Scale=&Lerp_Float(A=0.8, B=1, Alpha=$System.FN_Contrail_Setup.Dive Value)
        Bias=0
       )
       B=&FloatFromCurve(
        FloatCurve=DataInterface() {propertyValues {Curve {defaultValue=0, keys [{tangentMode=RCTM_User, time=0, value=0}, {tangentMode=RCTM_User, time=1, value=1}]}}}
        CurveIndex=$Particles.Normalized Spawn Group
       )
      )
     )
    )) {isSetParameters=true, module=SetVariables_#4}
   /SetVariables(Particles.Position=&AddVectorToPosition(
     Position=$Particles.Position
     Vector=&Multiply_VectorByFloat(
      Vector=&SelectVectorFromArray(
       Direct Array Index=&SpawnGroup()
       Direct Array Mode=::NewEnumerator1
       Vector Selection Array=DataInterface() {propertyValues {FloatData [<1 0 -0.5>, <-1 0 -0.5>]}}
      )
      Float=&FloatFromCurve(
       FloatCurve=DataInterface() {propertyValues {Curve {keys [
           {arriveTangent=4.1785, interpMode=RCIM_Cubic, leaveTangent=4.1785, tangentMode=RCTM_User, time=0, value=0}
           {arriveTangent=1.1179, interpMode=RCIM_Cubic, leaveTangent=1.1179, tangentMode=RCTM_User, time=0.2042, value=0.6228}
           {arriveTangent=0.0224, interpMode=RCIM_Cubic, leaveTangent=0.0224, tangentMode=RCTM_User, time=1, value=1}
          ]}}}
       CurveIndex=$Particles.Normalized Spawn Group
       Scale Curve=10
      )
     )
    )) {isSetParameters=true}
  ]
  rendererClasses [&NiagaraRibbonRendererProperties]
  renderers [{rendererClass=&NiagaraRibbonRendererProperties}]
 }
 Balloons {
  emitterUpdateScript [
   &EmitterState(
    Inactive Response=::NewEnumerator1
    Life Cycle Mode=::NewEnumerator1
    Loop Behavior=::NewEnumerator1
    Loop Duration Mode=::NewEnumerator1
    UseLoopDelay=false
   )
   &SpawnBurst_Instantaneous(Spawn Count=2, Use Spawn Probability=false, UseLoopCountLimit=false)
  ]
  particleSpawnScript [
   &PartitionParticles()
   &InitializeParticle(
    Color Mode=::NewEnumerator0
    Lifetime=5
    Mesh Scale Mode=::NewEnumerator3
    Mesh Uniform Scale=1.45
    UsePositionOffset=false
    Write Material Random=true
   )
   /SetVariables(Particles.VisibilityTag=$Particles.PartitionParticles.Partition) {isSetParameters=true}
   &DynamicMaterialParameters(
    Index 0 Param 1=&MergeHelper_FloatSwitch(
     0=0
     1=&RandomRangeFloat(Minimum=0.25, Maximum=0.75)
     2=?
     3=?
     Index=$Particles.PartitionParticles.Partition
     Num=::NewEnumerator0
    )
    Index 0 Param 2=$Particles.Timeline.Time
    Index 0 Param 3=$User.ParaGlideLeanAlpha
    Index 0 Param 4=&RandomRangeFloat(Minimum=2, Maximum=3)
    Module.Index0Param0Write=false
    Module.Index0Param1Write=false
    Module.Index0Param2Write=false
    Module.Index0Param3Write=true
   ) {module=DynamicMaterialParameters002}
  ]
  particleUpdateScript [
   &ParticleState(Kill Particles When Lifetime Has Elapsed=false, Let Infinitely Lived Particles Die When Emitter Deactivates=true)
   &SkeletalMeshLocation(
    Bone / Socket Index=$Particles.PartitionParticles.Partition
    Bone Sampling Mode=::NewEnumerator4
    MaskSpawnGroup=false
    Position Sampling=::NewEnumerator1
    Sampled Position Offset=<0 0 5>
    Sampled Position Offset Coordinate Space:ENiagaraSkelMeshTransforms=::NewEnumerator1
    Skeletal Mesh=$System.Skeletal Mesh
   )
   /SetVariables(Particles.Balloon_Orientation:Quat4f=&MultiplyQuaternion(
     Quaternion A=&MakeQuatFromAxisAngle(
      Axis=<0 0 1>
      AngleInDegrees=&SelectFloatFromArray(Direct Array Index=$Particles.PartitionParticles.Partition, Float Selection Array=DataInterface() {propertyValues {FloatData [5,-10]}})
     )
     Quaternion B=&MakeQuatFromAxisAngle(
      Axis=<0 0 -1>
      AngleInDegrees=&Lerp_Float(
       A=0
       B=&SelectFloatFromArray(Direct Array Index=$Particles.PartitionParticles.Partition, Float Selection Array=DataInterface() {propertyValues {FloatData [-50,50]}})
       Alpha=$System.FN_Contrail_Setup.Dive Value
      )
      NormalizeAxis=true
     )
    )) {isSetParameters=true}
   /SetVariables(Particles.Balloon_Orientation:Quat4f=&MultiplyQuaternion(
     Quaternion A=$Particles.Balloon_Orientation
     Quaternion B=&MakeQuatFromAxisAngle(
      Axis=&SelectVectorFromArray(
       Direct Array Index=$Particles.PartitionParticles.Partition
       Vector Selection Array=DataInterface() {propertyValues {FloatData [<-1 1 0>, <1 1 0>]}}
      )
      AngleInDegrees=&Lerp_Float(
       A=0
       B=&MakeCustomFloatFromBool(
        Bool=$User.bIsFrontEndPreview
        True Float=&SelectFloatFromArray(Direct Array Index=$Particles.PartitionParticles.Partition, Float Selection Array=DataInterface() {propertyValues {FloatData [-50,50]}})
        False Float=0
       )
       Alpha=$System.FN_Contrail_Setup.Dive Value
      )
      NormalizeAxis=true
     )
    )) {isSetParameters=true}
   /SetVariables(Particles.Balloon_Position:NiagaraPosition=$Particles.SkeletalMeshLocation.SampledPosition) {isSetParameters=true}
   &Timeline(
    Play Rate=&Lerp_Float(A=1, B=2, Alpha=&ClampFloat(Float=$User.ParaGlideLeanAlpha))
    Timeline End=10000000
   )
   &DynamicMaterialParameters(
    Index 0 Param 1=&MergeHelper_FloatSwitch(
     0=0
     1=&RandomRangeFloat(Minimum=0.25, Maximum=0.75)
     2=?
     3=?
     Index=$Particles.PartitionParticles.Partition
     Num=::NewEnumerator0
    )
    Index 0 Param 2=$Particles.Timeline.Time
    Index 0 Param 3=$User.ParaGlideLeanAlpha
    Module.Index0Float0Write=true
    Module.Index0Float1Write=true
    Module.Index0Float2Write=true
    Module.Index0Float3Write=true
   ) {module=DynamicMaterialParameters001}
   /SetVariables(Particles.OutlineSpherePos:NiagaraPosition=&AddVectorToPosition(
     Position=$Particles.SkeletalMeshLocation.SampledPosition
     Vector=&MergeHelper_VectorSwitch(0=<7.5 0 85>, 1=<-5 0 85>, 2=?, 3=?, Index=$Particles.PartitionParticles.Partition, Num=::NewEnumerator0)
    )) {isSetParameters=true}
  ]
  rendererClasses [&NiagaraMeshRendererProperties]
  renderers [
   {rendererClass=&NiagaraMeshRendererProperties, rendererIndex=0}
   {rendererClass=&NiagaraMeshRendererProperties, rendererIndex=1}
   {rendererClass=&NiagaraMeshRendererProperties, rendererIndex=2}
   {rendererClass=&NiagaraMeshRendererProperties, rendererIndex=3}
  ]
 }
 ConfettiSquares {
  emitterUpdateScript [
   &EmitterState()
   &SpawnRate(
    Spawn Probability=0.33
    SpawnRate=&Lerp_Float(A=15, B=30, Alpha=$System.FN_Contrail_Setup.Dive Value)
    Use Spawn Probability=false
   )
  ]
  particleSpawnScript [
   &FN_SetSpawnGroup(Number of Spawn Groups=2)
   &InitializeParticle(
    Lifetime Max=1.25
    Lifetime Mode=::NewEnumerator1
    Sprite Size Mode=::NewEnumerator1
    Uniform Sprite Size Max=6
    Uniform Sprite Size Min=3
    UsePositionOffset=false
    Write Material Random=true
   )
   &SkeletalMeshLocation(
    Bone / Socket Index=$Particles.SpawnGroup
    Bone Sampling Mode=::NewEnumerator4
    Interpolate Mesh Sampling=false
    MaskSpawnGroup=false
    Sampled Position Offset Coordinate Space:ENiagaraSkelMeshTransforms=::NewEnumerator2
    Skeletal Mesh=$System.Skeletal Mesh
   ) {module=SkeletalMeshLocation001}
   &ShapeLocation(
    MaskSpawnGroup=false
    Offset=&Lerp_Vector(
     A=&MergeHelper_VectorSwitch(0=<7.5 0 -10>, 1=<-7.5 0 -10>, 2=?, 3=?, Index=$Particles.SpawnGroup, Num=::NewEnumerator0)
     B=&MergeHelper_VectorSwitch(0=<7.5 15 -10>, 1=<-7.5 15 -10>, 2=?, 3=?, Index=$Particles.SpawnGroup, Num=::NewEnumerator0)
     Alpha=$System.FN_Contrail_Setup.Dive Value
    )
    Offset Mode=::NewEnumerator0
    Override Seed=false
    Sphere Radius=&Lerp_Float(A=40, B=70, Alpha=$System.FN_Contrail_Setup.Dive Value)
    Sphere Surface Distribution=1
    UseOwnerScale=true
   )
   &AddVelocity(
    Cone Angle=1
    Cone Axis=&FN_FrontEndVector3Modifier(Gameplay=<0 0 1>, Front End=&Lerp_Vector(A=<0 0 1>, B=<0 -1 0.75>, Alpha=$System.FN_Contrail_Setup.Dive Value))
    Module.UseConeFalloff=false
    Module.UseVelDistribution=false
    Override Seed=false
    Velocity Mode=::NewEnumerator2
    Velocity Speed=&Lerp_Float(A=200, B=400, Alpha=$System.FN_Contrail_Setup.Dive Value)
   ) {module=AddVelocity001}
  ]
  particleUpdateScript [
   &ParticleState()
   &GravityForce(Gravity=<0 0 -111>)
   /Niagara/Modules/Update/Forces/CurlNoiseForce(Noise Frequency=0.02, Noise Strength=150, Pan Noise=false)
   &Drag(
    Drag=&FloatFromCurve(FloatCurve=DataInterface() {propertyValues {Curve {keys [<at 0 0.5>,<at 1 1>]}}}, Scale Curve=0.4)
    Use Linear Drag=true
    Use Rotational Drag=true
   )
   &SolveForcesAndVelocity(Clamp Velocity=false, Limit Acceleration=false)
   &ScaleSpriteSize(Scale Factor=&Vector2DFromFloat(Value=&FloatFromCurve(
      FloatCurve=DataInterface() {propertyValues {Curve {keys [
          <at 0 0 RCIM_Cubic>
          {arriveTangent=-0.0623, interpMode=RCIM_Cubic, leaveTangent=-0.0623, tangentMode=RCTM_User, time=0.1, value=1}
          {arriveTangent=-0.0083, interpMode=RCIM_Cubic, leaveTangent=-0.0083, tangentMode=RCTM_User, time=0.5, value=1}
          <at 1 0 RCIM_Cubic>
         ]}}}
      Scale Curve=&Multiply_Float(
       A=1
       B=/Niagara/DynamicInputs/Camera/CameraFade(
        Camera Query:NiagaraDataInterfaceCamera=DataInterface()
        Fade Offset:NiagaraFloat=0
        Fade Distance:NiagaraFloat=150
        Use Owner Position:NiagaraBool=true
       )
      )
     )))
   &SubUVAnimation(
    "Sprite Renderer "=DataInterface() {propertyValues {SpriteRenderer=&self:ConfettiSquares_0.NiagaraSpriteRendererProperties_0}}
    Start Frame Offset:NiagaraInt32=&RandomRangeInt(Minimum=0, Maximum=555)
    UseEndFrame=false
    UseStartFrame=false
   )
   &ScaleColor(
    Scale RGB=&VectorFromFloat(Value=&FNTimeOfDayScalar(Night Scale=0.25, Morning Scale=0.333, Day Scale=0.5, Evening Scale=0.25))
    ScaleA=true
    ScaleRGB=true
   )
  ]
  rendererClasses [&NiagaraSpriteRendererProperties]
  renderers [{rendererClass=&NiagaraSpriteRendererProperties}]
 }
 HandWraps {
  emitterUpdateScript [
   &EmitterState(
    Inactive Response=::NewEnumerator1
    Life Cycle Mode=::NewEnumerator1
    Loop Behavior=::NewEnumerator1
    Loop Duration Mode=::NewEnumerator1
    UseLoopDelay=true
   )
   &SpawnBurst_Instantaneous(Spawn Count=12, Use Spawn Probability=false, UseLoopCountLimit=false)
  ]
  particleSpawnScript [
   &PartitionParticles()
   &InitializeParticle(
    Color=<rgba 0.1 0.1 0.1 1>
    Mesh Scale Max=<0.275 0.175 0.275>
    Mesh Scale Min=<0.175 0.175 0.175>
    Mesh Scale Mode=::NewEnumerator2
    UsePositionOffset=false
    Write Material Random=true
   )
  ]
  particleUpdateScript [
   &ParticleState(Kill Particles When Lifetime Has Elapsed=false, Let Infinitely Lived Particles Die When Emitter Deactivates=true)
   &SkeletalMeshLocation(
    Bone / Socket Index=$Particles.PartitionParticles.Partition
    Bone Sampling Mode=::NewEnumerator4
    MaskSpawnGroup=false
    Skeletal Mesh=$System.Skeletal Mesh
   )
   /SetVariables(Particles.Balloon_Orientation:Quat4f=&MultiplyQuaternion(
     Quaternion A=&MakeQuatFromAxisAngle(
      Axis=<1 0 0>
      AngleInDegrees=&SelectFloatFromArray(Direct Array Index=$Particles.PartitionParticles.Partition, Float Selection Array=DataInterface() {propertyValues {FloatData [90,90]}})
     )
     Quaternion B=&MakeQuatFromAxisAngle(
      Axis=<1 0 0>
      AngleInDegrees=&Lerp_Float(
       A=0
       B=&SelectFloatFromArray(Direct Array Index=$Particles.PartitionParticles.Partition, Float Selection Array=DataInterface() {propertyValues {FloatData [-30,-30]}})
       Alpha=$System.FN_Contrail_Setup.Dive Value
      )
      NormalizeAxis=true
     )
    )) {isSetParameters=true}
   /SetVariables(Particles.Balloon_Orientation:Quat4f=&MultiplyQuaternion(
     Quaternion A=$Particles.Balloon_Orientation
     Quaternion B=&MakeQuatFromAxisAngle(
      Axis=&SelectVectorFromArray(
       Direct Array Index=$Particles.PartitionParticles.Partition
       Vector Selection Array=DataInterface() {propertyValues {FloatData [<0 0 0>, <0 0 0>]}}
      )
      AngleInDegrees=&Lerp_Float(
       A=0
       B=&MakeCustomFloatFromBool(
        Bool=$User.bIsFrontEndPreview
        True Float=&SelectFloatFromArray(Direct Array Index=$Particles.PartitionParticles.Partition, Float Selection Array=DataInterface() {propertyValues {FloatData [0,0]}})
        False Float=0
       )
       Alpha=$System.FN_Contrail_Setup.Dive Value
      )
      NormalizeAxis=true
     )
    )) {isSetParameters=true}
   /SetVariables(Particles.Balloon_Position:NiagaraPosition=$Particles.SkeletalMeshLocation.SampledPosition) {isSetParameters=true}
   &DynamicMaterialParameters(
    Index 0 Param 1=&RandomRangeFloat(Minimum=5, Maximum=8)
    Index 0 Param 2=&RandomRangeFloat(Minimum=5, Maximum=8)
    Index 0 Param 3=1
    Module.Param0WriteEnabled=true
    Module.Param1WriteEnabled=true
    Module.Param2WriteEnabled=true
    Module.Param3WriteEnabled=true
   ) {module=DynamicMaterialParameters002}
   &CameraOffset(Camera Offset Amount=2)
   &SolveForcesAndVelocity(Clamp Velocity=false, Limit Acceleration=false)
  ]
  rendererClasses [&NiagaraMeshRendererProperties]
  renderers [{rendererClass=&NiagaraMeshRendererProperties, rendererIndex=0}]
 }
}
summary {
 systemName=NS_Contrail_PajamaSoar
 emitters [
  {emitterName=Balloons, rendererClasses [&NiagaraMeshRendererProperties]}
  {emitterName=HandWraps, rendererClasses [&NiagaraMeshRendererProperties]}
  {emitterName=BallonWire_Hand, rendererClasses [&NiagaraRibbonRendererProperties]}
  {emitterName=ConfettiSquares, rendererClasses [&NiagaraSpriteRendererProperties]}
 ]
 userVariables [
  User.bIsFrontEnd {type=NiagaraBool}
  User.bIsFrontEndPreview {type=NiagaraBool}
  User.ParaGlideLeanAlpha {type=NiagaraFloat, defaultValue=0, description=""}
  User.PlayerSpeed {type=NiagaraFloat, defaultValue=0, description=""}
  User.SkeletalMesh {
   type=NiagaraDataInterfaceSkeletalMesh
   defaultValue {
    struct=NiagaraExt_VariableValue_DataInterface
    value {dataInterface=&self:NiagaraDataInterfaceSkeletalMesh_0}
   }
  }
 ]
}
